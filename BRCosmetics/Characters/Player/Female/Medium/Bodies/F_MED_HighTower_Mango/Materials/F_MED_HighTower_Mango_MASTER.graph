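# Master material for the F_MED_HighTower_Mango female character body, built on material attributes for skinned meshes with morph targets.
# The shared PreFX character function feeds a chain of static-switched layers: cloth fuzz, silk, clear coat, metal LUT and holo foil borrowed from HighTower_Squash.
# SkinFX_Mask channels drive the first four layers, while holo foil uses pre-skinned position and normal.
# An always-on emissive fresnel follows, then the shared PostFX function finishes the result.
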Material /BRCosmetics/Characters/Player/Female/Medium/Bodies/F_MED_HighTower_Mango/Materials/F_MED_HighTower_Mango_MASTER
graph {
 properties {
  bAutomaticallySetUsageInEditor=false
  bUseMaterialAttributes=true
  bUsedWithMeshDeformer=true
  bUsedWithMorphTargets=true
  bUsedWithSkeletalMesh=true
  pixelDepthOffsetMode=PDOM_Legacy
  shadingModel=MSM_FromMaterialExpression
  usedShadingModels=Default Lit | Clear Coat | Subsurface Profile | Cloth
 }
 nodes [
  n1 = MaterialFunctionCall() {materialFunction=&BaseFN_CharacterFunction_PreFX}
  n2 = TextureSampleParameter2D() {parameterName=SkinFX_Mask}
  n3 = StaticSwitchParameter(
   True=MaterialFunctionCall!(
    Alpha=StaticComponentMaskParameter(None=@n2.RGB) {defaultR=true, group=Skin FX, parameterName=Cloth Channel}
    Gmap=@n1.Gmap
    Material=@n1.Gmap
   ) {materialFunction=/Game/Weapons/WeaponSkins/MaterialFunction/ClothFuzz}.Result
   False=@n1.Result
  ) {group=Skin FX, parameterName=Use Cloth fuzz}
  n4 = StaticSwitchParameter(
   True=MaterialFunctionCall!(Alpha=ChannelMaskParameter(None=@n2.RGB) {group=Skin FX, parameterName=Silk Channel}.R, Material=@n3) {materialFunction=/Game/Weapons/WeaponSkins/MaterialFunction/MF_Silk}.Result
   False=@n3
  ) {group=Skin FX, parameterName=Use Silk}
  n5 = StaticSwitchParameter(
   True=MaterialFunctionCall!(Alpha=StaticComponentMaskParameter(None=@n2.RGB) {defaultR=true, group=Skin FX, parameterName=Clear Coat Channel}, In=@n4) {materialFunction=/Game/Weapons/WeaponSkins/MaterialFunction/Clearcoat}.Result
   False=@n4
  ) {group=Skin FX, parameterName=Use Clear Coat}
  n6 = StaticSwitchParameter(
   True=MaterialFunctionCall(Mask=@n2.RGBA, Materials=@n5) {materialFunction=&MF_MetalsLUT}.Result
   False=@n5
  ) {group=Skin FX, parameterName=UseMetalLUT}
  n7 = MaterialFunctionCall!(PreskinnedLocalPosition=@n1.Pre skinned Local Position, PreskinnedLocalNormal=VertexInterpolator(VS=PreSkinnedNormal()).PS, Material=@n6) {materialFunction=/Game/Characters/Player/Female/Medium/Bodies/F_MED_HighTower_Squash/Materials/MF_HoloFoil}
  MaterialFunctionCall(
   AO=@n1.AO
   Distance Blend=@n1.AO
   Fresnel=@n1.AO
   Gmap=@n1.AO
   Material=MaterialFunctionCall!(In=StaticSwitchParameter(True=@n7.Result, False=@n6) {parameterName=BUseHoloFoil}) {materialFunction=/Game/Weapons/WeaponSkins/MaterialFunction/MF_EmissiveFresnel}.Result
  ) {materialFunction=&BaseFN_CharacterFunction_PostFX}
 ]
}
