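# Physics asset for the rear cloth trail (perf variant) on the F_MED_HighTower_Mango female character body.
# A five-bone capsule chain hangs from a small parent body constrained to the pelvis, with swing and tighter twist limits between links.
# Pelvis, thigh, calf and right-foot capsules supply body shapes with free angular constraints; there is no left-foot body.
# A large offset sphere, Floor_COL, approximates the ground as a collision shape.
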PhysicsAsset /BRCosmetics/Characters/Player/Female/Medium/Bodies/F_MED_HighTower_Mango/Meshes/F_MED_HighTower_Mango_ClothRear_perf_Physics
bodies {
 Floor_COL [{shapeType=Sphere, radius=300, shapeName=Floor_COL_sphere, center=<0 -20 -295>}]
 calf_l [{shapeType=Capsule, length=44, radius=7.75, center=<-27.1847 -2.8308 -0.8779>, rotation=<rot 90 90 88.9456>}]
 calf_r [{shapeType=Capsule, length=40, radius=8.5, center=<27.8724 2.3738 0.7116>, rotation=<rot -90 -90 89.1261>}]
 dyn_parent_perf_trail_bck [{shapeType=Capsule, length=0.505, radius=0.505, shapeName=dyn_parent_perf_trail_bck_capsule, rotation=<rot 0 -90 0>}]
 dyn_perf_trail_bck_01 [{shapeType=Capsule, length=9.5, radius=7, shapeName=dyn_perf_trail_bck_01_capsule, center=<7.3947 0 -5>, rotation=<rot 0 0 90>}]
 dyn_perf_trail_bck_02 [{shapeType=Capsule, length=8.75, radius=6.75, shapeName=dyn_perf_trail_bck_02_capsule, center=<6 0 -5.5>, rotation=<rot 0 0 90>}]
 dyn_perf_trail_bck_03 [{shapeType=Capsule, length=6, radius=8, shapeName=dyn_perf_trail_bck_03_capsule, center=<7.399 0 -5.5>, rotation=<rot 0 0 90>}]
 dyn_perf_trail_bck_04 [{shapeType=Capsule, length=8.5, radius=7.75, shapeName=dyn_perf_trail_bck_04_capsule, center=<6 0 -5>, rotation=<rot 0 0 90>}]
 dyn_perf_trail_bck_05 [{shapeType=Capsule, length=12.25, radius=8, shapeName=dyn_perf_trail_bck_05_capsule, center=<6.6007 0 -6>, rotation=<rot 0 0 90>}]
 foot_r [{
   shapeType=Capsule
   length=13.5738
   radius=7.3311
   shapeName=foot_r_capsule
   center=<4.4057 -5.2969 0.0283>
   rotation=<rot 0.8965 17.6563 -86.6521>
  }]
 pelvis [{
   shapeType=Capsule
   length=13.9543
   radius=8.9543
   shapeName=pelvis_capsule
   center=<-4.0634 -2.8715 0.0322>
   rotation=<rot 0.1832 -90 0.1119>
  }]
 thigh_l [{shapeType=Capsule, length=32.9913, radius=8.5405, center=<-23.4379 -1.5228 -1.0327>, rotation=<rot -0.7433 -90.8397 -94.5301>}]
 thigh_r [{shapeType=Capsule, length=42.7413, radius=7.7905, center=<21.4162 1.0601 -0.2351>, rotation=<rot -0.6626 -89.8236 -94.5426>}]
}
constraints [
 {
  bone1Name=calf_l
  bone2Name=thigh_l
  swing1LimitDegrees=45
  swing1Motion=Free
  swing2LimitDegrees=45
  swing2Motion=Free
  twistLimitDegrees=45
  twistMotion=Free
 }
 {
  bone1Name=calf_r
  bone2Name=thigh_r
  swing1LimitDegrees=45
  swing1Motion=Free
  swing2LimitDegrees=45
  swing2Motion=Free
  twistLimitDegrees=45
  twistMotion=Free
 }
 {bone1Name=dyn_perf_trail_bck_01, bone2Name=dyn_parent_perf_trail_bck, swing1LimitDegrees=45, swing2LimitDegrees=45, twistLimitDegrees=15}
 {bone1Name=dyn_perf_trail_bck_02, bone2Name=dyn_perf_trail_bck_01, swing1LimitDegrees=45, swing2LimitDegrees=45, twistLimitDegrees=15}
 {bone1Name=dyn_perf_trail_bck_03, bone2Name=dyn_perf_trail_bck_02, swing1LimitDegrees=45, swing2LimitDegrees=45, twistLimitDegrees=15}
 {bone1Name=dyn_perf_trail_bck_04, bone2Name=dyn_perf_trail_bck_03, swing1LimitDegrees=45, swing2LimitDegrees=45, twistLimitDegrees=15}
 {bone1Name=dyn_perf_trail_bck_05, bone2Name=dyn_perf_trail_bck_04, swing1LimitDegrees=45, swing2LimitDegrees=45, twistLimitDegrees=15}
 {
  bone1Name=foot_r
  bone2Name=calf_r
  swing1LimitDegrees=45
  swing1Motion=Free
  swing2LimitDegrees=45
  swing2Motion=Free
  twistLimitDegrees=45
  twistMotion=Free
 }
 {
  bone1Name=thigh_l
  bone2Name=pelvis
  swing1LimitDegrees=45
  swing1Motion=Free
  swing2LimitDegrees=45
  swing2Motion=Free
  twistLimitDegrees=45
  twistMotion=Free
 }
 {
  bone1Name=thigh_r
  bone2Name=pelvis
  swing1LimitDegrees=45
  swing1Motion=Free
  swing2LimitDegrees=45
  swing2Motion=Free
  twistLimitDegrees=45
  twistMotion=Free
 }
 {
  bone1Name=dyn_parent_perf_trail_bck
  bone2Name=pelvis
  swing1LimitDegrees=45
  swing1Motion=Free
  swing2LimitDegrees=45
  swing2Motion=Free
  twistLimitDegrees=45
  twistMotion=Free
 }
]
