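# Physics asset for the F_MED_HighTower_Mango female player body, covering a front loincloth and the lower body around it.
# Two four-box chains, left and right, branch from a middle cloth box that hangs off a small driver capsule linked to the pelvis.
# Cross constraints tie each left-right pair together, forming a ladder that approximates a cloth sheet.
# Capsules cover pelvis, spine, thighs and calves, joined by free-motion constraints.
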PhysicsAsset /BRCosmetics/Characters/Player/Female/Medium/Bodies/F_MED_HighTower_Mango/Meshes/F_MED_HighTower_Mango_Physics
bodies {
 calf_l [{shapeType=Capsule, length=41.5, radius=7.75, center=<-27.1847 -2.8308 -0.8779>, rotation=<rot 90 90 88.9456>}]
 calf_r [{shapeType=Capsule, length=37.5, radius=8.5, center=<27.8724 2.3738 0.7116>, rotation=<rot -90 -90 89.1261>}]
 dyn_cloth_font_mid_01 [{shapeType=Box, extentX=14.138, extentY=12.0184, extentZ=1.3776, center=<7.4134 0 0>}]
 dyn_driver_cloth_front [{shapeType=Capsule, length=0.505, radius=0.505, shapeName=dyn_driver_cloth_front_capsule, rotation=<rot 0 -90 0>}]
 dyn_loincloth_01_l [{shapeType=Box, extentX=11.7301, extentY=9.8395, extentZ=2.8776, center=<5.6591 3 -0.0001>}]
 dyn_loincloth_01_r [{shapeType=Box, extentX=11.7301, extentY=11.6363, extentZ=3.3776, center=<-5.6591 -3.9 0>}]
 dyn_loincloth_02_l [{shapeType=Box, extentX=11.7301, extentY=11.7752, extentZ=2.8776, center=<7 3.9 0>}]
 dyn_loincloth_02_r [{shapeType=Box, extentX=11.7301, extentY=11.1363, extentZ=2.8776, center=<-7 -3.6 0>}]
 dyn_loincloth_03_l [{shapeType=Box, extentX=16.2301, extentY=14.6363, extentZ=1.8776, center=<5.65 3.75 0>}]
 dyn_loincloth_03_r [{shapeType=Box, extentX=16.2301, extentY=14.6363, extentZ=1.8776, center=<-5.65 -4.4 0>}]
 dyn_loincloth_04_l [{shapeType=Box, extentX=21.2301, extentY=16.1363, extentZ=1.8776, center=<11.1511 4.2 0>}]
 dyn_loincloth_04_r [{shapeType=Box, extentX=21.2301, extentY=14.6363, extentZ=1.8776, center=<-11.0258 -3.75 0>}]
 pelvis [{shapeType=Capsule, length=37.26, radius=11.99, center=<-17.9292 -6.2241 0>, rotation=<rot -90 0 -1.7629>}]
 spine_02 [{shapeType=Capsule, length=2, radius=2}]
 thigh_l [{shapeType=Capsule, length=41.6314, radius=8.25, center=<-23.989 -1.7186 -0.0002>, rotation=<rot 90 0 -1.0629>}]
 thigh_r [{shapeType=Capsule, length=41.3009, radius=9.2636, center=<23.989 1.7186 -0.0002>, rotation=<rot 90 0 -1.0629>}]
}
constraints [
 {bone1Name=dyn_loincloth_01_l, bone2Name=dyn_cloth_font_mid_01, swing1LimitDegrees=35, swing2LimitDegrees=30, twistLimitDegrees=45}
 {bone1Name=dyn_loincloth_02_l, bone2Name=dyn_loincloth_01_l, swing1LimitDegrees=45, swing2LimitDegrees=45, twistLimitDegrees=70}
 {bone1Name=dyn_loincloth_03_l, bone2Name=dyn_loincloth_02_l, swing1LimitDegrees=55, swing2LimitDegrees=45, twistLimitDegrees=70}
 {bone1Name=dyn_loincloth_01_r, bone2Name=dyn_cloth_font_mid_01, swing1LimitDegrees=35, swing2LimitDegrees=30, twistLimitDegrees=45}
 {bone1Name=dyn_loincloth_02_r, bone2Name=dyn_loincloth_01_r, swing1LimitDegrees=45, swing2LimitDegrees=45, twistLimitDegrees=70}
 {bone1Name=dyn_loincloth_03_r, bone2Name=dyn_loincloth_02_r, swing1LimitDegrees=55, swing2LimitDegrees=45, twistLimitDegrees=70}
 {
  bone1Name=spine_02
  bone2Name=pelvis
  swing1LimitDegrees=45
  swing1Motion=Free
  swing2LimitDegrees=45
  swing2Motion=Free
  twistLimitDegrees=45
  twistMotion=Free
 }
 {bone1Name=dyn_cloth_font_mid_01, bone2Name=dyn_driver_cloth_front, swing1LimitDegrees=30, swing2LimitDegrees=70, twistLimitDegrees=20}
 {
  bone1Name=thigh_l
  bone2Name=pelvis
  swing1LimitDegrees=45
  swing1Motion=Free
  swing2LimitDegrees=45
  swing2Motion=Free
  twistLimitDegrees=45
  twistMotion=Free
 }
 {
  bone1Name=thigh_r
  bone2Name=pelvis
  swing1LimitDegrees=45
  swing1Motion=Free
  swing2LimitDegrees=45
  swing2Motion=Free
  twistLimitDegrees=45
  twistMotion=Free
 }
 {bone1Name=dyn_loincloth_04_l, bone2Name=dyn_loincloth_03_l, swing1LimitDegrees=55, swing2LimitDegrees=45, twistLimitDegrees=70}
 {bone1Name=dyn_loincloth_04_r, bone2Name=dyn_loincloth_03_r, swing1LimitDegrees=55, swing2LimitDegrees=45, twistLimitDegrees=70}
 {
  bone1Name=calf_r
  bone2Name=thigh_r
  swing1LimitDegrees=45
  swing1Motion=Free
  swing2LimitDegrees=45
  swing2Motion=Free
  twistLimitDegrees=45
  twistMotion=Free
 }
 {
  bone1Name=calf_l
  bone2Name=thigh_l
  swing1LimitDegrees=45
  swing1Motion=Free
  swing2LimitDegrees=45
  swing2Motion=Free
  twistLimitDegrees=45
  twistMotion=Free
 }
 {bone1Name=dyn_loincloth_01_r, bone2Name=dyn_loincloth_01_l, swing1LimitDegrees=45, swing2LimitDegrees=45, twistLimitDegrees=45}
 {bone1Name=dyn_loincloth_02_r, bone2Name=dyn_loincloth_02_l, swing1LimitDegrees=45, swing2LimitDegrees=45, twistLimitDegrees=45}
 {bone1Name=dyn_loincloth_03_r, bone2Name=dyn_loincloth_03_l, swing1LimitDegrees=45, swing2LimitDegrees=45, twistLimitDegrees=45}
 {bone1Name=dyn_loincloth_04_r, bone2Name=dyn_loincloth_04_l, swing1LimitDegrees=45, swing2LimitDegrees=45, twistLimitDegrees=45}
 {
  bone1Name=dyn_driver_cloth_front
  bone2Name=pelvis
  swing1LimitDegrees=45
  swing1Motion=Free
  swing2LimitDegrees=45
  swing2Motion=Free
  twistLimitDegrees=45
  twistMotion=Free
 }
]
