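# Post-process style AnimBP for the HighTower Mango female body, adding secondary motion to skirt, loincloth and skull bones.
# It copies the parent mesh pose, then IsinFrontEndVis selects front-end rigid-body simulation or cached in-game dynamics.
# The in-game path chains thigh-driven bone offsets, two RigidBody sims, a weighted skull constraint and a skydive-weighted Control Rig.
# Events update gravity and alphas; the trail-controller cloth branch never reaches the output.
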AnimBlueprint /BRCosmetics/Characters/Player/Female/Medium/Bodies/F_MED_HighTower_Mango/Meshes/F_MED_HighTower_Mango_AnimBP_Pre
target_skeleton=&/F_MED_HighTower_Mango_Skeleton
graphs [
 AnimGraph {nodes [
   n1 = "Transform (Modify) Bone - Bone: dyn_parent_cloth_trail_bck"(ComponentPose=Local To Component(LocalPose=Use cached pose 'AnimCache'()), Alpha=1, Translation="0,0,0", Rotation="0,-18,0", Scale="1,1,1") {
    rotationMode=BMM_Additive
    rotationSpace=BCS_BoneSpace
    alphaBoolBlend {blendOption=Linear}
    boneToModify {boneName=dyn_parent_cloth_trail_bck}
    rotation {yaw=-18}
   }
   n2 = "Transform (Modify) Bone - Bone: dyn_parent_cloth_trail_frt"(ComponentPose=@n1, Alpha=1, Translation="0,0,0", Rotation="0,-12,0", Scale="1,1,1") {
    rotationMode=BMM_Additive
    rotationSpace=BCS_BoneSpace
    alphaBoolBlend {blendOption=Linear}
    boneToModify {boneName=dyn_parent_cloth_trail_frt}
    rotation {yaw=-12}
   }
   n3 = Trail controller - Bone: dyn_cloth_trail_frt_06(ComponentPose=@n2, Alpha=Get FwdMovement() -> (FwdMovement:double)) {
    bActorSpaceFakeVel=true
    bUsePlanarLimit=true
    chainLength=5
    relaxationSpeedScale=1.5
    baseJoint {boneName=root}
    planarLimits [{drivingBone {boneName=pelvis}, planeTransform {location=<-35 5 0>, rotation=<rot 0 -6 90>, scale=<1 1 1>}}]
    rotationLimits [{limitMax=<180 180 180>}, {limitMax=<180 180 180>}, {limitMax=<180 180 180>}, {limitMax=<180 180 180>}, {limitMax=<180 180 180>}]
    rotationOffsets [<0 0 0>, <0 0 0>, <0 0 0>, <0 0 0>, <0 0 0>]
    trailBone {boneName=dyn_cloth_trail_frt_06}
   }
   n4 = Trail controller - Bone: dyn_cloth_trail_bck_05(ComponentPose=@n3, Alpha=Get BwdMovement() -> (BwdMovement:double)) {
    bActorSpaceFakeVel=true
    bUsePlanarLimit=true
    chainLength=5
    relaxationSpeedScale=1.2
    baseJoint {boneName=root}
    planarLimits [{drivingBone {boneName=pelvis}, planeTransform {location=<-35 -10 0>, rotation=<rot 0 0 -90>, scale=<1 1 1>}}]
    rotationLimits [{limitMax=<180 180 180>}, {limitMax=<180 180 180>}, {limitMax=<180 180 180>}, {limitMax=<180 180 180>}, {limitMax=<180 180 180>}]
    rotationOffsets [<0 0 0>, <0 0 0>, <0 0 0>, <0 0 0>, <0 0 0>]
    trailBone {boneName=dyn_cloth_trail_bck_05}
   }
   n5 = "Constraint - dyn_cloth_font_mid_01 (P)"(ComponentPose=@n4, Alpha=1, ConstraintWeights_0=1) {
    alphaBoolBlend {blendOption=Linear}
    boneToModify {boneName=dyn_cloth_font_mid_01}
    constraintSetup [{targetBone {boneName=dyn_cloth_trail_frt_01}}]
    constraintWeights [1]
   }
   n6 = "Constraint - dyn_loincloth_01_l (P)"(ComponentPose=@n5, Alpha=1, ConstraintWeights_0=1) {
    alphaBoolBlend {blendOption=Linear}
    boneToModify {boneName=dyn_loincloth_01_l}
    constraintSetup [{targetBone {boneName=dyn_cloth_trail_frt_01}}]
    constraintWeights [1]
   }
   n7 = "Constraint - dyn_loincloth_01_r (P)"(ComponentPose=@n6, Alpha=1, ConstraintWeights_0=1) {
    alphaBoolBlend {blendOption=Linear}
    boneToModify {boneName=dyn_loincloth_01_r}
    constraintSetup [{targetBone {boneName=dyn_cloth_trail_frt_01}}]
    constraintWeights [1]
   }
   n8 = "Constraint - dyn_loincloth_01_l (P)"(ComponentPose=@n7, Alpha=1, ConstraintWeights_0=1) {
    alphaBoolBlend {blendOption=Linear}
    boneToModify {boneName=dyn_loincloth_01_l}
    constraintSetup [{targetBone {boneName=dyn_cloth_trail_frt_02}}]
    constraintWeights [1]
   }
   n9 = "Constraint - dyn_loincloth_02_r (P)"(ComponentPose=@n8, Alpha=1, ConstraintWeights_0=1) {
    alphaBoolBlend {blendOption=Linear}
    boneToModify {boneName=dyn_loincloth_02_r}
    constraintSetup [{targetBone {boneName=dyn_cloth_trail_frt_02}}]
    constraintWeights [1]
   }
   n10 = "Constraint - dyn_loincloth_03_l (P)"(ComponentPose=@n9, Alpha=1, ConstraintWeights_0=1) {
    alphaBoolBlend {blendOption=Linear}
    boneToModify {boneName=dyn_loincloth_03_l}
    constraintSetup [{targetBone {boneName=dyn_cloth_trail_frt_03}}]
    constraintWeights [1]
   }
   n11 = "Constraint - dyn_loincloth_03_r (P)"(ComponentPose=@n10, Alpha=1, ConstraintWeights_0=1) {
    alphaBoolBlend {blendOption=Linear}
    boneToModify {boneName=dyn_loincloth_03_r}
    constraintSetup [{targetBone {boneName=dyn_cloth_trail_frt_03}}]
    constraintWeights [1]
   }
   n12 = "Constraint - dyn_loincloth_04_r (P)"(ComponentPose=@n11, Alpha=1, ConstraintWeights_0=1) {
    alphaBoolBlend {blendOption=Linear}
    boneToModify {boneName=dyn_loincloth_04_r}
    constraintSetup [{targetBone {boneName=dyn_cloth_trail_frt_04}}]
    constraintWeights [1]
   }
   n13 = "Constraint - dyn_loincloth_04_r (P)"(ComponentPose=@n12, Alpha=1, ConstraintWeights_0=1) {
    alphaBoolBlend {blendOption=Linear}
    boneToModify {boneName=dyn_loincloth_04_r}
    constraintSetup [{targetBone {boneName=dyn_cloth_trail_frt_04}}]
    constraintWeights [1]
   }
   n14 = "Constraint - dyn_skirt_rear_01_l (P)"(ComponentPose=@n13, Alpha=1, ConstraintWeights_0=1) {
    alphaBoolBlend {blendOption=Linear}
    boneToModify {boneName=dyn_skirt_rear_01_l}
    constraintSetup [{targetBone {boneName=dyn_cloth_trail_bck_01}}]
    constraintWeights [1]
   }
   n15 = "Constraint - dyn_skirt_rear_01_r (P)"(ComponentPose=@n14, Alpha=1, ConstraintWeights_0=1) {
    alphaBoolBlend {blendOption=Linear}
    boneToModify {boneName=dyn_skirt_rear_01_r}
    constraintSetup [{targetBone {boneName=dyn_cloth_trail_bck_01}}]
    constraintWeights [1]
   }
   n16 = "Constraint - dyn_skirt_rear_02_l (P)"(ComponentPose=@n15, Alpha=1, ConstraintWeights_0=1) {
    alphaBoolBlend {blendOption=Linear}
    boneToModify {boneName=dyn_skirt_rear_02_l}
    constraintSetup [{targetBone {boneName=dyn_cloth_trail_bck_02}}]
    constraintWeights [1]
   }
   n17 = "Constraint - dyn_skirt_rear_02_r (P)"(ComponentPose=@n16, Alpha=1, ConstraintWeights_0=1) {
    alphaBoolBlend {blendOption=Linear}
    boneToModify {boneName=dyn_skirt_rear_02_r}
    constraintSetup [{targetBone {boneName=dyn_cloth_trail_bck_02}}]
    constraintWeights [1]
   }
   n18 = "Constraint - dyn_skirt_rear_03_l (P)"(ComponentPose=@n17, Alpha=1, ConstraintWeights_0=1) {
    alphaBoolBlend {blendOption=Linear}
    boneToModify {boneName=dyn_skirt_rear_03_l}
    constraintSetup [{targetBone {boneName=dyn_cloth_trail_bck_03}}]
    constraintWeights [1]
   }
   n19 = "Constraint - dyn_skirt_rear_03_r (P)"(ComponentPose=@n18, Alpha=1, ConstraintWeights_0=1) {
    alphaBoolBlend {blendOption=Linear}
    boneToModify {boneName=dyn_skirt_rear_03_r}
    constraintSetup [{targetBone {boneName=dyn_cloth_trail_bck_03}}]
    constraintWeights [1]
   }
   n20 = "Constraint - dyn_skirt_rear_04_l (P)"(ComponentPose=@n19, Alpha=1, ConstraintWeights_0=1) {
    alphaBoolBlend {blendOption=Linear}
    boneToModify {boneName=dyn_skirt_rear_04_l}
    constraintSetup [{targetBone {boneName=dyn_cloth_trail_bck_04}}]
    constraintWeights [1]
   }
   n21 = "Constraint - dyn_skirt_rear_04_r (P)"(ComponentPose=@n20, Alpha=1, ConstraintWeights_0=1) {
    alphaBoolBlend {blendOption=Linear}
    boneToModify {boneName=dyn_skirt_rear_04_r}
    constraintSetup [{targetBone {boneName=dyn_cloth_trail_bck_04}}]
    constraintWeights [1]
   }
   Blend Poses by bool(BlendPose_1=Component To Local(ComponentPose=@n21), bActiveValue=Get bRigidBodyEnabled()) {blendType=Linear}
   "Constraint - dyn_loincloth_01_r (P)"(Alpha=1, ConstraintWeights_0=1) {
    alphaBoolBlend {blendOption=Linear}
    boneToModify {boneName=dyn_loincloth_01_r}
    constraintSetup [{targetBone {boneName=dyn_cloth_trail_frt_01}}]
    constraintWeights [1]
   }
   n22 = RigidBody!(
    ComponentPose=Local To Component(LocalPose=Copy Pose From Mesh() {bCopyCurves=true, bUseAttachedParent=true})
    Alpha
    OverrideWorldGravity:Vector=Get GravityOverride_b() -> (GravityOverride_b:Vector)
    ExternalForce=Get VectorForce() -> (VectorForce:Vector)
   ) -> (Pose) {
    bOverrideWorldGravity=true
    overridePhysicsAsset=&/F_MED_HighTower_Mango_ClothRear_perf_Physics
    simulationSpace=BaseBoneSpace
    alphaBoolBlend {blendOption=Linear}
    alphaScaleBiasClamp {bInterpResult=true, interpSpeedDecreasing=1, interpSpeedIncreasing=1}
    baseBoneRef {boneName=root}
    componentAppliedLinearAccClamp=<1000 1000 1500>
    componentLinearAccScale=<0.7 0.7 0.75>
    componentLinearVelScale=<0.7 0.7 0.75>
    overrideWorldGravity {z=-920}
   }
   n23 = RigidBody(ComponentPose=@n22, OverrideWorldGravity=Get GravityOverride_b(), ExternalForce=Get VectorForce()) {
    bOverrideWorldGravity=true
    overridePhysicsAsset=&/F_MED_HighTower_Mango_FrontEnd_Physics
    simulationSpace=BaseBoneSpace
    alphaBoolBlend {blendOption=Linear}
    baseBoneRef {boneName=pelvis}
    componentAppliedLinearAccClamp=<3000 3000 3000>
    componentLinearAccScale=<0.25 0.25 0.25>
    componentLinearVelScale=<0.3 0.3 0.3>
    overrideWorldGravity {z=-900}
    simSpaceSettings {
     maxAngularAcceleration=1000
     maxAngularVelocity=4
     maxLinearAcceleration=700
     maxLinearVelocity=500
     velocityScaleZ=0.6
     externalLinearDragV {x=0.2}
    }
   }
   Output Pose(Result=Blend Poses by bool(
     BlendPose_0=Component To Local(ComponentPose=@n23)
     BlendPose_1=Use cached pose 'InGameDyanamics'()
     bActiveValue=Get IsinFrontEndVis() -> (IsinFrontEndVis:bool)
    ) {blendType=Linear}) {layerGroup=None}
   n24 = "Transform (Modify) Bone - Bone: dyn_driver_cloth_front"(
    ComponentPose=Local To Component(LocalPose=Use cached pose 'AnimCache'())
    Alpha=Get ThighRotationLeft() -> (ThighRotationLeft:float)
    Translation="-7,2,0"
    Rotation="0,0,0"
    Scale="1,1,1"
   ) {
    rotationMode=BMM_Additive
    rotationSpace=BCS_BoneSpace
    translationMode=BMM_Additive
    translationSpace=BCS_BoneSpace
    alphaBoolBlend {blendOption=Linear}
    boneToModify {boneName=dyn_driver_cloth_front}
    translation=<-7 2>
   }
   n25 = RigidBody(ComponentPose=@n24, OverrideWorldGravity=Get GravityOverride_b(), ExternalForce=Get VectorForce()) {
    bOverrideWorldGravity=true
    overridePhysicsAsset=&F_MED_HighTower_Mango_ClothRear_perf_Physics
    simulationSpace=BaseBoneSpace
    alphaBoolBlend {blendOption=Linear}
    alphaScaleBiasClamp {bInterpResult=true, interpSpeedDecreasing=1, interpSpeedIncreasing=1}
    baseBoneRef {boneName=root}
    componentAppliedLinearAccClamp=<1000 1000 1500>
    componentLinearAccScale=<0.7 0.7 0.75>
    componentLinearVelScale=<0.7 0.7 0.75>
    overrideWorldGravity {z=-920}
   }
   n26 = RigidBody(ComponentPose=@n25, OverrideWorldGravity=Get GravityOverride_b(), ExternalForce=Get VectorForce()) {
    bOverrideWorldGravity=true
    overridePhysicsAsset=&/F_MED_HighTower_Mango_Physics
    simulationSpace=BaseBoneSpace
    alphaBoolBlend {blendOption=Linear}
    baseBoneRef {boneName=root}
    componentAppliedLinearAccClamp=<1100 1100 1100>
    componentLinearAccScale=<0.4 0.4 0.4>
    componentLinearVelScale=<0.4 0.4 0.4>
    overrideWorldGravity {z=-900}
   }
   n27 = Get front_end_rear_alpha() -> (front_end_rear_alpha:double)
   n28 = "Transform (Modify) Bone - Bone: driver_skirt_rear_01_l"(ComponentPose=@n26, Alpha=@n27, Translation="0,0,0", Rotation="0,-2.5,0", Scale="1,1,1") {
    rotationMode=BMM_Additive
    rotationSpace=BCS_BoneSpace
    translationMode=BMM_Additive
    translationSpace=BCS_BoneSpace
    alphaBoolBlend {blendOption=Linear}
    boneToModify {boneName=driver_skirt_rear_01_l}
    rotation {yaw=-2.5}
   }
   n29 = "Transform (Modify) Bone - Bone: driver_skirt_rear_01_r"(ComponentPose=@n28, Alpha=@n27, Translation="0,0,0", Rotation="0,-2.5,0", Scale="1,1,1") {
    rotationMode=BMM_Additive
    rotationSpace=BCS_BoneSpace
    translationMode=BMM_Additive
    translationSpace=BCS_BoneSpace
    alphaBoolBlend {blendOption=Linear}
    boneToModify {boneName=driver_skirt_rear_01_r}
    rotation {yaw=-2.5}
   }
   n30 = Get front_end_alpha() -> (front_end_alpha:double)
   n31 = "Transform (Modify) Bone - Bone: driver_skirt_rear_01_l"(ComponentPose=@n29, Alpha=@n30, Translation="0,0,0", Rotation="0,-5,0", Scale="1,1,1") {
    rotationMode=BMM_Additive
    rotationSpace=BCS_BoneSpace
    translationMode=BMM_Additive
    translationSpace=BCS_BoneSpace
    alphaBoolBlend {blendOption=Linear}
    boneToModify {boneName=driver_skirt_rear_01_l}
    rotation {yaw=-5}
   }
   n32 = "Transform (Modify) Bone - Bone: driver_skirt_rear_01_r"(ComponentPose=@n31, Alpha=@n30, Translation="0,0,0", Rotation="0,-2,0", Scale="1,1,1") {
    rotationMode=BMM_Additive
    rotationSpace=BCS_BoneSpace
    translationMode=BMM_Additive
    translationSpace=BCS_BoneSpace
    alphaBoolBlend {blendOption=Linear}
    boneToModify {boneName=driver_skirt_rear_01_r}
    rotation {yaw=-2}
   }
   n33 = "Constraint - dyn_skull_aim (PPPPPP)"!(
    ComponentPose=Local To Component(LocalPose=Component To Local(ComponentPose=@n32))
    Alpha=1
    ConstraintWeights_0=0.0107
    ConstraintWeights_1:float=0.3095
    ConstraintWeights_2:float=0.2914
    ConstraintWeights_3:float=0.1787
    ConstraintWeights_4:float=0.0103
    ConstraintWeights_5:float=0.1994
   ) -> (Pose) {
    alphaBoolBlend {blendOption=Linear}
    boneToModify {boneName=dyn_skull_aim}
    constraintSetup [
     {targetBone {boneName=dyn_loincloth_02_r}}
     {targetBone {boneName=dyn_loincloth_01_l}}
     {targetBone {boneName=dyn_cloth_font_mid_01}}
     {targetBone {boneName=pelvis}}
     {targetBone {boneName=dyn_loincloth_02_l}}
     {targetBone {boneName=dyn_loincloth_01_r}}
    ]
    constraintWeights [1,1,1,1,1,1]
   }
   n34 = Control Rig(Source=Component To Local(ComponentPose=@n33), Alpha=Get skydive_active() -> (skydive_active:double)) {
    alphaBoolBlend {blendOption=Linear}
    alphaScaleBiasClamp {bInterpResult=true, interpSpeedDecreasing=15, interpSpeedIncreasing=15}
    controlRigAssetReference {blueprintRigClass=&/F_MED_HighTower_Mango_CtrlRIg.F_MED_HighTower_Mango_CtrlRIg_C}
   }
   Save cached pose 'InGameDyanamics'(Pose=@n34) {cacheName=InGameDyanamics}
   n35 = Get front_end_rear_alpha()
   n36 = "Transform (Modify) Bone - Bone: driver_skirt_rear_01_l"(
    ComponentPose=Local To Component(LocalPose=Copy Pose From Mesh() {bCopyCurves=true, bUseAttachedParent=true})
    Alpha=@n35
    Translation="0,-2.5,0"
    Rotation="0,-5,0"
    Scale="1,1,1"
   ) {
    rotationMode=BMM_Additive
    rotationSpace=BCS_BoneSpace
    translationMode=BMM_Additive
    translationSpace=BCS_BoneSpace
    alphaBoolBlend {blendOption=Linear}
    boneToModify {boneName=driver_skirt_rear_01_l}
    rotation {yaw=-5}
    translation {y=-2.5}
   }
   n37 = "Transform (Modify) Bone - Bone: driver_skirt_rear_01_r"(ComponentPose=@n36, Alpha=@n35, Translation="0,-2.5,0", Rotation="0,-5,0", Scale="1,1,1") {
    rotationMode=BMM_Additive
    rotationSpace=BCS_BoneSpace
    translationMode=BMM_Additive
    translationSpace=BCS_BoneSpace
    alphaBoolBlend {blendOption=Linear}
    boneToModify {boneName=driver_skirt_rear_01_r}
    rotation {yaw=-5}
    translation {y=-2.5}
   }
   n38 = "Transform (Modify) Bone - Bone: dyn_glute_l"(ComponentPose=@n37, Alpha=Get ThighRotationLeft(), Translation="-3.750,0.875,-0.9", Rotation="0.000,0.000,-55", Scale="1.25,1.25,1.25") {
    rotationMode=BMM_Additive
    rotationSpace=BCS_BoneSpace
    scaleMode=BMM_Additive
    scaleSpace=BCS_BoneSpace
    translationMode=BMM_Additive
    translationSpace=BCS_BoneSpace
    alphaBoolBlend {blendOption=Linear}
    boneToModify {boneName=dyn_glute_l}
    rotation {roll=-55}
    scale=<1.25 1.25 1.25>
    translation=<-3.75 0.875 -0.9>
   }
   n39 = "Transform (Modify) Bone - Bone: dyn_glute_r"(
    ComponentPose=@n38
    Alpha=Get ThighRotationRight() -> (ThighRotationRight:float)
    Translation="-3.75,0,0"
    Rotation="0.000,0.000,-55"
    Scale="1.3,1.3,1.3"
   ) {
    rotationMode=BMM_Additive
    rotationSpace=BCS_BoneSpace
    scaleMode=BMM_Additive
    scaleSpace=BCS_BoneSpace
    translationMode=BMM_Additive
    translationSpace=BCS_BoneSpace
    alphaBoolBlend {blendOption=Linear}
    boneToModify {boneName=dyn_glute_r}
    rotation {roll=-55}
    scale=<1.3 1.3 1.3>
    translation {x=-3.75}
   }
   n40 = "Transform (Modify) Bone - Bone: Floor_COL"!(
    ComponentPose=@n39
    bAlphaBoolEnabled:bool=OR Boolean!(A=Get bIsFalling(), B=Get bIsSliding(), C:bool=Get bIsDBNO()) -> (ReturnValue)
    Translation="0,0,-300"
    Rotation="0,0,0"
    Scale="1,1,1"
   ) -> (Pose) {
    alphaInputType=Bool
    translationMode=BMM_Additive
    translationSpace=BCS_BoneSpace
    alphaBoolBlend {blendInTime=0.1, blendOption=Linear, blendOutTime=0.1}
    boneToModify {boneName=Floor_COL}
    translation {z=-300}
   }
   Save cached pose 'AnimCache'(Pose=Component To Local(ComponentPose=@n40)) {cacheName=AnimCache}
   Get GravityOverride()
   Get ClothAlpha() -> (ClothAlpha:float)
   Get GravityOverride()
  ]}
 EventGraph {nodes [
   Event BlueprintInitializeAnimation()
   > Set Skel_Mesh(Skel_Mesh=GetPartSkeletalMeshComponent())
   > n1 = Get!(execute:Exec) -> (then:Exec, else:Exec, Skel_Mesh:Skeletal Mesh Component Object Reference)
   > Set Anim_Inst(Anim_Inst:Anim Instance Object Reference=GetAnimInstance(self=@n1)) -> (Output_Get:Anim Instance Object Reference)
   Event BlueprintBeginPlay() {is=title K2Node_Event}
   > Set IsinFrontEndVis(IsinFrontEndVis:bool=Get bIsInFrontEnd()) -> (Output_Get:bool)
   n2 = Event OnNotifyPartsStopped!() -> (OutputDelegate, then, Context:Name=None) {is=title K2Node_Event}
   > Branch(Condition="Equal (Name)"!(A:Name=@n2.Context, B:Name=Female_Commando_Idle_02) -> (ReturnValue:bool) {is=title K2Node_CallFunction})
   > Set front_end_rear_alpha(front_end_rear_alpha:double=0) -> (Output_Get:double)
   Event OnBeginSkydiving()
   > Set skydive_active(skydive_active:double=1) -> (Output_Get:double)
   Event OnCharacterCustomizationLoaded()
   n3 = CheckForBackbling()
   > Set wearing_backbling(wearing_backbling:bool=@n3) -> (Output_Get:bool)
   > Set wearing_long_cape(wearing_long_cape:bool=@n3.bIsBackblingLong) -> (Output_Get:bool)
   > Set bIsWearingLongCape(bIsWearingLongCape:bool=@n3.bIsBackblingLong) -> (Output_Get:bool)
   Event OnEndSkydiving()
   > Set skydive_active(skydive_active=0.1)
   n4 = Event OnNotifyPartsStarted!() -> (OutputDelegate, then, Context:Name=None) {is=title K2Node_Event}
   > Branch(Condition="Equal (Name)"(A=@n4.Context, B=Female_Commando_Idle_02))
   > Set front_end_rear_alpha(front_end_rear_alpha=1)
   Event BlueprintUpdateAnimation()
   > Get()
   > Get!(execute) -> (then, else, Anim_Inst:Anim Instance Object Reference)
   > n5 = Sequence!(execute) -> (then_0, then_1, then_2:Exec, then_3:Exec, then_4:Exec, then_5:Exec, then_6:Exec, then_7:Exec)
   n6 = Branch(execute=@n5.then_1, Condition=Get bIsSwimSprinting() -> (bIsSwimSprinting:bool))
   > Set GravityOverride_b(GravityOverride_b:Vector="0, 0, 0") -> (Output_Get:Vector)
   n7 = Branch(execute=@n6.else, Condition=Get bIsSkyDiving())
   > n8 = Branch(Condition=Get bIsParachuteOpened())
   Set GravityOverride_b(execute=[@n8.then, @n7.else], GravityOverride_b="0, 0,-920")
   n9 = Branch(execute=@n8.else, Condition="float > float"(A=Get PawnVelocityZ(), B=50) {is=title K2Node_CallFunction})
   > Set GravityOverride_b(GravityOverride_b="0, 0,200")
   n10 = PerlinNoise1D!(Value:float=GetGameTimeInSeconds()) -> (ReturnValue:float)
   n11 = float * float(A="Sin (Degrees)"!(A:double=float * float(A=@n10, B=500)) -> (ReturnValue:double), B=-400)
   n12 = Get Skel_Mesh()
   n13 = GetSocketRotation!(self:Scene Component Object Reference=@n12, InSocketName:Name=pelvis) -> (ReturnValue:Rotator)
   n14 = vector + vector(
    A:Vector=vector * float(A=GetForwardVector!(InRot:Rotator=@n13) -> (ReturnValue:Vector), B=-800)
    B:Vector=MakeVector(Y=0, Z=float + float(A=@n11, B=200) {is=title K2Node_CommutativeAssociativeBinaryOperator})
   ) -> (ReturnValue:Vector) {is=title K2Node_CommutativeAssociativeBinaryOperator}
   Set GravityOverride_b(execute=@n9.else, GravityOverride_b=@n14)
   n15 = Self-Reference()
   n16 = PlayerMovement(execute=@n5.then_5, AnimInstance=@n15, XDotProductCoefficient=1, YDotProductCoefficient=1, ZDotProductCoefficient=1)
   n17 = MapRangeClamped(Value=@n16.YMovement, InRangeA=0, InRangeB=-1)
   n18 = float * float(A="Absolute (Float)"(A=@n16), B=0.5)
   > Set BwdMovement(BwdMovement:double=MapRangeClamped(Value=float + float(A=@n18, B=@n17), InRangeA=0, InRangeB=1.1, OutRangeA=0.5, OutRangeB=0)) -> (Output_Get:double)
   > n19 = PlayerMovement(AnimInstance=@n15, XDotProductCoefficient=1, YDotProductCoefficient=1, ZDotProductCoefficient=1)
   n20 = MapRangeClamped(Value=@n19.YMovement, InRangeA=0, InRangeB=1)
   n21 = float * float(A="Absolute (Float)"(A=@n19), B=0.5)
   > Set FwdMovement(FwdMovement:double=MapRangeClamped(Value=float + float(A=@n21, B=@n20), InRangeA=0, InRangeB=1.1, OutRangeA=0.5, OutRangeB=0)) -> (Output_Get:double)
   n22 = Branch(execute=@n5.then_6, Condition=Get bIsSkyDiving())
   > n23 = Branch(Condition="float > float"(A=Get PawnVelocityZ(), B=50))
   > Set BwdMovement(BwdMovement=float * float(A=Get BwdMovement(), B=0))
   Set BwdMovement(execute=[@n22.else, @n23.else], BwdMovement=float * float(A=Get BwdMovement(), B=1))
   n24 = Branch(
    execute=@n5.then_2
    Condition=Montage_IsPlaying(
     self=Get Anim_Inst() -> (Anim_Inst:Anim Instance Object Reference)
     Montage=/BRCosmetics/Animation/Game/MainPlayer/Emotes/ShaolinSipUp/Emote_ShaolinSipUp_CMF_M
    )
   )
   > Set emote_shaolin_sip_active(emote_shaolin_sip_active:double=1) -> (Output_Get:double)
   Set emote_shaolin_sip_active(execute=@n24.else, emote_shaolin_sip_active=0)
   n25 = Get PawnSpeed2D()
   n26 = Branch(
    execute=@n5.then_0
    Condition="float <= float"!(A:double=MapRangeClamped(Value=@n25, InRangeA=0, InRangeB=150), B:double=0) -> (ReturnValue:bool) {is=title K2Node_CallFunction}
   )
   > Set ClothAlpha_b(ClothAlpha_b:double=0.45) -> (Output_Get:double)
   > Set side_cloth_alpha(side_cloth_alpha:double=0.7) -> (Output_Get:double)
   > n27 = Branch(Condition=Get wearing_long_cape() -> (wearing_long_cape:bool))
   > n28 = Set ClothAlpha_b(ClothAlpha_b=0.25)
   n29 = Branch(execute=[@n28.then, @n27.else], Condition=Get bIsCrouching())
   > Set crouch_skirt_alpha(crouch_skirt_alpha:double=1) -> (Output_Get:double)
   > Set side_cloth_alpha(side_cloth_alpha=0.6)
   Set crouch_skirt_alpha(execute=@n29.else, crouch_skirt_alpha=0)
   > n30 = Branch(Condition=Get emote_active() -> (emote_active:bool))
   > n31 = Set ClothAlpha_b(ClothAlpha_b=1)
   Branch(execute=[@n30.else, @n31.then], Condition=Get bCanPlayCustomAnimations() -> (bCanPlayCustomAnimations:bool))
   > Set ClothAlpha_b(ClothAlpha_b=1)
   Set ClothAlpha_b(execute=@n26.else, ClothAlpha_b=1)
   > Set side_cloth_alpha(side_cloth_alpha=0.7)
   > n32 = Branch(Condition=Get bIsCrouching())
   > n33 = Set crouch_skirt_alpha(crouch_skirt_alpha=0.1)
   Branch(execute=[@n34.then, @n33.then], Condition=Get wearing_long_cape())
   > Set ClothAlpha_b(ClothAlpha_b=0.3)
   n34 = Set crouch_skirt_alpha(execute=@n32.else, crouch_skirt_alpha=0)
   Set thigh_up_alpha(
    execute=@n5.then_3
    thigh_up_alpha:double=GetAlphaFromReferenceTransformDelta(
     SkeletalMeshComponent=Get Skel_Mesh()
     BoneName=thigh_r
     BaseName=pelvis
     DeltaComponentType=QuaternionTwist
     InputRange=(X=0,Y=100)
     TwistAxis="0,0,1"
    )
   ) -> (Output_Get:double)
   > n35 = Branch(Condition=Get bIsInFrontEnd())
   > Set front_end_alpha(front_end_alpha:double=1) -> (Output_Get:double)
   Set front_end_alpha(execute=@n35.else, front_end_alpha=0)
   n36 = Get Skel_Mesh()
   n37 = GetAlphaFromReferenceTransformDelta(
    SkeletalMeshComponent=@n36
    BoneName=thigh_l
    BaseName=pelvis
    DeltaComponentType=QuaternionTwist
    InputRange=(X=20,Y=100)
    TwistAxis="0,0,1"
   )
   Set ThighRotationLeft_b(execute=@n5.then_4, ThighRotationLeft_b:double=@n37) -> (Output_Get:double)
   n38 = GetAlphaFromReferenceTransformDelta(
    SkeletalMeshComponent=@n36
    BoneName=thigh_r
    BaseName=pelvis
    DeltaComponentType=QuaternionTwist
    InputRange=(X=20,Y=100)
    TwistAxis="0,0,1"
   )
   > Set ThighRotationRight_b(ThighRotationRight_b:double=@n38) -> (Output_Get:double)
   > Branch(Condition="float > float"(A=Get PawnSpeed2D(), B=100))
   > Set ThighRotationLeft_b(ThighRotationLeft_b=float * float(A=@n37, B=0.35))
   > Set ThighRotationRight_b(ThighRotationRight_b=float * float(A=@n38, B=0.35))
   Event OnBeginPlayingEmote() {is=title K2Node_Event}
   > Set emote_active(emote_active:bool=true) -> (Output_Get:bool)
   Event OnEndPlayingEmote() {is=title K2Node_Event}
   > Set emote_active(emote_active=false)
   n39 = Branch(Condition=Get bIsSkyDiving())
   > n40 = Branch(Condition=Get bIsParachuteOpened())
   > Set GravityOverride_b(GravityOverride_b="0, 0,-800")
   n41 = Break Fort Anim Input Skydiving External Force(FortAnimInput_SkydivingExternalForce=Get SkydivingExternalForceInput()) -> (
    bUseSkydivingVectorForce=~
    SkydivingVectorForce=~
    HeadToPelvisDirection=~
    FloatingMultiplier=~
    FloatingAdditive=~
    DivingMultiplier=~
    DivingAdditive=~
    ParachutingMultiplier=~
    ParachutingAdditive=~
   )
   > Set VectorForce_b(VectorForce_b:Vector=@n41.SkydivingVectorForce) -> (Output_Get:Vector)
   n42 = Branch(execute=@n40.else, Condition=Get bIsSkydiveFloating())
   n43 = Branch(execute=[@n42.else, @n42.then], Condition="float > float"(A=Get PawnVelocityZ(), B=50))
   > Set GravityOverride_b(GravityOverride_b="0, 0,-300")
   > Set VectorForce_b(VectorForce_b=@n41.SkydivingVectorForce)
   Set GravityOverride_b(execute=@n43.else, GravityOverride_b="0, 0,-200")
   > Set VectorForce_b(VectorForce_b=@n41.SkydivingVectorForce)
   > Branch(Condition=Get bCanPlayCustomAnimations())
   > Set GravityOverride_b(GravityOverride_b="0, 0,150")
   Set GravityOverride_b(execute=@n39.else, GravityOverride_b="0, 0,-920")
   > Set VectorForce_b(VectorForce_b="0, 0, 0")
   GetOwnerSkeletalMeshComponent()
   n44 = GetSocketLocation!(self:Scene Component Object Reference=@n12, InSocketName:Name=pelvis) -> (ReturnValue:Vector)
   DrawDebugLine!(
    execute
    LineStart:Vector=@n44
    LineEnd:Vector=vector + vector(A=@n44, B=@n14)
    LineColor:Linear Color Structure=(R=1,G=0,B=0,A=1)
    Duration:float=0
    Thickness:float=3
    DepthPriority:EDrawDebugSceneDepthPriorityGroup Enum=World
   ) -> (then)
  ]}
]
variables [
 Anim_Inst {type=AnimInstance, default=null}
 BwdMovement {type=number}
 ClothAlpha_b {type=number}
 crouch_skirt_alpha {type=number}
 emote_active {type=boolean}
 emote_shaolin_sip_active {type=number}
 front_end_alpha {type=number}
 front_end_rear_alpha {type=number}
 FwdMovement {type=number}
 GravityOverride_b {type=Vector}
 IsinFrontEndVis {type=boolean}
 side_cloth_alpha {type=number}
 Skel_Mesh {type=SkeletalMeshComponent}
 skydive_active {type=number, default=0.1}
 thigh_up_alpha {type=number}
 ThighRotationLeft_b {type=number}
 ThighRotationRight_b {type=number}
 VectorForce_b {type=Vector}
 wearing_backbling {type=boolean}
 wearing_long_cape {type=boolean}
]
