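# Animation Blueprint for the F_MED_HighTower_Mango female player body, adding secondary motion to its hair.
# Caches a pose copied from the attached parent, with upper-arm branches layered from Body_Mesh.
# When bRigidBodyEnabled is true, that pose gets an additive pose, a head/clavicle constraint, corrective bone offsets driven by arm and head rotation, and AnimDynamics on front-right hair.
# A neck-relative RigidBody node with a hair physics asset finishes either branch.
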AnimBlueprint /BRCosmetics/Characters/Player/Female/Medium/Bodies/F_MED_HighTower_Mango/Meshes/Parts/F_MED_HighTower_Mango_AnimBP
target_skeleton=&/F_MED_HighTower_Mango_Skeleton
graphs [
 AnimGraph {nodes [
   n1 = "Sequence Player (additive) 'F_MED_HighTower_Mango_Skeleton_Pose'"() {sequence=&/F_MED_HighTower_Mango_Skeleton_Pose}
   n2 = Use cached pose 'AnimCache'()
   n3 = "Constraint - dyn_hair_fr_r_02 (PP)"!(
    ComponentPose=Local To Component(LocalPose=Apply Additive(Base=@n2, Additive=@n1) {alphaBoolBlend {blendOption=Linear}})
    Alpha=1
    ConstraintWeights_0=0.8
    ConstraintWeights_1:float=0.2
   ) -> (Pose) {
    alphaBoolBlend {blendOption=Linear}
    boneToModify {boneName=dyn_hair_fr_r_02}
    constraintSetup [{targetBone {boneName=clavicle_r}}, {targetBone {boneName=head}}]
    constraintWeights [1,1]
   }
   n4 = "Transform (Modify) Bone - Bone: dyn_hair_r_01"(ComponentPose=@n3, Alpha=Get Arm Rot() -> (Arm Rot:double), Translation="-3,-0.000,0.000", Rotation="15,0,0", Scale="1,1,1") {
    rotationMode=BMM_Additive
    rotationSpace=BCS_BoneSpace
    translationMode=BMM_Additive
    translationSpace=BCS_BoneSpace
    alphaBoolBlend {blendOption=Linear}
    boneToModify {boneName=dyn_hair_r_01}
    rotation {pitch=15}
    translation {x=-3}
   }
   n5 = "Transform (Modify) Bone - Bone: dyn_hair_fr_r_02"(ComponentPose=@n4, Alpha=Get head delta_fwd() -> (head delta_fwd:double), Translation="0,0,0", Rotation="0.000,-10,-0.000", Scale="1,1,1") {
    rotationMode=BMM_Additive
    rotationSpace=BCS_BoneSpace
    translationMode=BMM_Additive
    translationSpace=BCS_BoneSpace
    alphaBoolBlend {blendOption=Linear}
    boneToModify {boneName=dyn_hair_fr_r_02}
    rotation {yaw=-10}
   }
   n6 = Anim Dynamics - Bone: dyn_hair_fr_r_03(ComponentPose=@n5, Alpha=1, ExternalForce="0,0,0") {
    angularDampingOverride=0.9999
    angularSpringConstant=1500
    bAngularSpring=true
    bGravityOverrideInSimSpace=true
    bOverrideAngularDamping=true
    simulationSpace=RootRelative
    alphaBoolBlend {blendOption=Linear}
    boundBone {boneName=dyn_hair_fr_r_03}
    componentAppliedLinearAccClamp=<3500 3500 3500>
    componentLinearAccScale=<0.6 0.6 0.8>
    componentLinearVelScale=<0.6 0.6 0.8>
    physicsBodyDefinitions [{
      sphereCollisionRadius=0
      boundBone {boneName=dyn_hair_fr_r_03}
      boxExtents=<20 20 10>
      constraintSetup {angularLimitsMax=<1 10 -5>, angularLimitsMin=<-1 -10 -30>, angularTarget=<0 0 0>}
      localJointOffset=<1 0 0>
     }]
    relativeSpaceBone {boneName=root}
   }
   n7 = Anim Dynamics - Bone: dyn_hair_fr_r_04(ComponentPose=@n6, Alpha=1, ExternalForce="0,0,0") {
    angularDampingOverride=0.9999
    angularSpringConstant=1500
    bAngularSpring=true
    bGravityOverrideInSimSpace=true
    bOverrideAngularDamping=true
    simulationSpace=RootRelative
    alphaBoolBlend {blendOption=Linear}
    boundBone {boneName=dyn_hair_fr_r_04}
    componentAppliedLinearAccClamp=<3500 3500 3500>
    componentLinearAccScale=<0.6 0.6 0.8>
    componentLinearVelScale=<0.6 0.6 0.8>
    physicsBodyDefinitions [{
      sphereCollisionRadius=0
      boundBone {boneName=dyn_hair_fr_r_04}
      boxExtents=<20 20 10>
      constraintSetup {angularLimitsMax=<1 10 -5>, angularLimitsMin=<-1 -10 -30>, angularTarget=<0 0 0>}
      localJointOffset=<1 0 0>
     }]
    relativeSpaceBone {boneName=root}
   }
   n8 = Component To Local(ComponentPose=@n7)
   n9 = Use cached pose 'AnimCache'()
   n10 = Get bRigidBodyEnabled()
   n11 = RigidBody(ComponentPose=Local To Component(LocalPose=Blend Poses by bool(BlendPose_0=@n8, BlendPose_1=@n9, bActiveValue=@n10) {blendType=Linear})) {
    bOverrideWorldGravity=true
    overridePhysicsAsset=&/F_MED_HighTower_Mango_Perf_Hair
    simulationSpace=BaseBoneSpace
    alphaBoolBlend {blendOption=Linear}
    baseBoneRef {boneName=neck_01}
    componentAppliedLinearAccClamp {y=1200}
    componentLinearAccScale=<0.3 0.35 0.3>
    componentLinearVelScale=<0.3 0.2 0.3>
    overrideWorldGravity {z=-980}
   }
   Output Pose(Result=Component To Local(ComponentPose=@n11)) {layerGroup=None}
   n12 = Get Body_Mesh()
   Save cached pose 'AnimCache'(Pose=Layered blend per bone(BasePose=Copy Pose From Mesh() {bUseAttachedParent=true}, BlendPoses_0=Copy Pose From Mesh(SourceMeshComponent=@n12)) {layerSetup [{branchFilters [{boneName=upperarm_l}, {boneName=upperarm_r}]}]}) {cacheName=AnimCache}
  ]}
 EventGraph {nodes [
   Event BlueprintUpdateAnimation()
   > Set Arm Rot(Arm Rot:double=GetAlphaFromReferenceTransformDelta(SkeletalMeshComponent=Get Skel_Mesh(), BoneName=upperarm_l, BaseName=spine_05, InputRange=(X=45,Y=80), TwistAxis="0,0,1")) -> (Output_Get:double)
   > Set head delta_fwd(head delta_fwd:double=GetAlphaFromReferenceTransformDelta(
     SkeletalMeshComponent=Get Skel_Mesh()
     BoneName=head
     BaseName=spine_05
     DeltaComponentType=QuaternionTwist
     InputRange=(X=0,Y=-20)
     TwistAxis="0,0,1"
    )) -> (Output_Get:double)
   Event BlueprintInitializeAnimation()
   n1 = GetOwnerSkeletalMeshComponent()
   > Set Skel_Mesh(Skel_Mesh=@n1)
   > Set Body_Mesh(Body_Mesh=@n1)
   > n2 = Cast To PlayerPawn_Athena(Object=TryGetPawnOwner())
   > n3 = Set Pawn Owner(Pawn Owner:Player Pawn Athena Object Reference=@n2) -> (Output_Get:Player Pawn Athena Object Reference)
   > Set Head_Mesh(Head_Mesh:Skeletal Mesh Component Object Reference=GetSkeletalMeshForPartType(self=@n3, PartType=Head)) -> (Output_Get:Skeletal Mesh Component Object Reference)
  ]}
]
variables [
 Arm Rot {type=number}
 Body_Mesh {type=SkeletalMeshComponent}
 head delta_fwd {type=number}
 Head_Mesh {type=SkeletalMeshComponent}
 Pawn Owner {type=PlayerPawn_Athena_C}
 Skel_Mesh {type=SkeletalMeshComponent}
]
